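; Gameplay ability for the Vine Berries augment that grants glider redeploy when the player gains a configured tag, noted as equipping a bush.
; On tag add it applies GE_GliderRedeploy to the owner and binds to the character's movement mode changes.
; Each change removes that effect by its handle unless the pawn has Gameplay.Status.ImmuneFallingDamage.
; The movement-mode switch and tag-removal branches are empty, so losing the tag does not itself remove the effect.
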(event Custom|MovementChanged
  (params (Character :type Character%20Object%20Reference) (PrevMovementMode :type EMovementMode%20Enum :default "MOVE_None") (PreviousCustomMode :type Byte))
  (switch Utilities|FlowControl|Switch|SwitchonEMovementMode PrevMovementMode
    (:MOVE_Walking
      _)
    (:MOVE_Falling
      _))
  (if (not (GameplayTags|HasMatchingGameplayTag (CallFunction|GetAvatarasFortPlayerPawn) "(TagName=\"Gameplay.Status.ImmuneFallingDamage\")"))
    (Ability|RemoveGameplayEffectFromOwnerWithHandle (stale Ability|EventActivateAbility _returnvalue))))

(event Ability|EventActivateAbility
  (bind _tag (Variables|Default|GetTag))
  ;[1 0.2718 0.06097] Activate when equipping bush
  (Ability|Tasks|WaitGameplayTagAdd _tag
    (:Added
      (bind _returnvalue (Ability|ApplyGameplayEffectToOwner "/Game/Athena/Items/SpyTech/Ability/GE_GliderRedeploy.GE_GliderRedeploy_C"))
      (Ability|Tasks|WaitGameplayTagRemove _tag
        (:then
          _)
        (:Removed
          _))
      (Character|BindEventtoMovementModeChangedDelegate (CallFunction|GetAvatarasFortPlayerPawn) (delegate Custom|MovementChanged)))))
